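; Blueprint function library for Creative devices that builds, merges, empty-checks and converts localizable payloads, which bundle plain text, a localizable message and a character limit.
; Conversions to string or text use the message when it is set, falling back to the plain text otherwise.
; Truncation applies only to message entries with a positive limit; text conversion skips the string round-trip when the limit is zero or less.
; Merging can retain existing payload content alongside default text.
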(fn CombineTextAndLocalizablePayload
  (params (bCanUseExistingPayload :type Boolean) (DefaultText :type Text) (ExistingPayload_Text_2_563E0DA1490D74CEE84EF39A5DF6BB26 :type Text) (ExistingPayload_Message_5_DDD38AFE499B5AE0E2151A99B81EE564 :type Localizable%20Message%20Structure) (ExistingPayload_CharLimit_11_F93CC8954D727984A85AACB9929A1685 :type Integer))
  (return (select (and bCanUseExistingPayload (not (LocalizableMessage|IsEmptyLocalizableMessage ExistingPayload_Message_5_DDD38AFE499B5AE0E2151A99B81EE564))) "" (select (and bCanUseExistingPayload (not (Utilities|Text|TextIsEmpty DefaultText))) ExistingPayload_Text_2_563E0DA1490D74CEE84EF39A5DF6BB26 DefaultText)) (select (and bCanUseExistingPayload (not (LocalizableMessage|IsEmptyLocalizableMessage ExistingPayload_Message_5_DDD38AFE499B5AE0E2151A99B81EE564))) ExistingPayload_Message_5_DDD38AFE499B5AE0E2151A99B81EE564 0) ExistingPayload_CharLimit_11_F93CC8954D727984A85AACB9929A1685))

(fn IsLocalizablePayloadEmpty
  (params (Payload_Text_2_563E0DA1490D74CEE84EF39A5DF6BB26 :type Text) (Payload_Message_5_DDD38AFE499B5AE0E2151A99B81EE564 :type Localizable%20Message%20Structure) (Payload_CharLimit_11_F93CC8954D727984A85AACB9929A1685 :type Integer))
  (return (and (Utilities|Text|TextIsEmpty Payload_Text_2_563E0DA1490D74CEE84EF39A5DF6BB26) (LocalizableMessage|IsEmptyLocalizableMessage Payload_Message_5_DDD38AFE499B5AE0E2151A99B81EE564))))

(fn LocalizableMessageToLocalizablePayload
  (params (Message :type Localizable%20Message%20Structure))
  (return (Utilities|Struct|MakeLocalizablePayload "" Message)))

(fn LocalizablePayloadToString
  (params (Payload :type Localizable%20Payload%20Structure))
  (return (Internal|LocalizablePayloadtoTruncatedString Payload (Variables|Get__WorldContext))))

(fn LocalizablePayloadToText
  (params (Payload :type Localizable%20Payload%20Structure))
  (Utilities|FlowControl|Sequence
    (:then_0
      (if (<= (Utilities|Struct|BreakLocalizablePayload Payload) 0)
        (LocalVariables|SetOutText (Internal|LocalizablePayloadtoUntruncatedText Payload (Variables|Get__WorldContext)))
        (else
          (LocalVariables|SetOutText (Utilities|Text|ToText(String) (Internal|LocalizablePayloadtoTruncatedString Payload (Variables|Get__WorldContext)))))))
    (:then_1
      (return (LocalVariables|GetOutText)))))

(fn LocalizablePayloadToTruncatedString
  (params (Payload_Text_2_563E0DA1490D74CEE84EF39A5DF6BB26 :type Text) (Payload_Message_5_DDD38AFE499B5AE0E2151A99B81EE564 :type Localizable%20Message%20Structure) (Payload_CharLimit_11_F93CC8954D727984A85AACB9929A1685 :type Integer))
  (Utilities|FlowControl|Sequence
    (:then_0
      (if (not (LocalizableMessage|IsEmptyLocalizableMessage Payload_Message_5_DDD38AFE499B5AE0E2151A99B81EE564))
        ;[1 0.06314 0.7165] For now, only enforce character limits on LocalizableMessage entries
        (LocalVariables|SetOutValue (Utilities|String|ToString(Text) (LocalizableMessage|ConvLocalizableMessagetoText (Variables|Get__WorldContext) Payload_Message_5_DDD38AFE499B5AE0E2151A99B81EE564)))
        (if (> Payload_CharLimit_11_F93CC8954D727984A85AACB9929A1685 0)
          (LocalVariables|SetOutValue (Utilities|String|Left (LocalVariables|GetOutValue) Payload_CharLimit_11_F93CC8954D727984A85AACB9929A1685)))
        (else
          (LocalVariables|SetOutValue (Utilities|String|ToString(Text) Payload_Text_2_563E0DA1490D74CEE84EF39A5DF6BB26)))))
    (:then_1
      (return (LocalVariables|GetOutValue)))))

(fn LocalizablePayloadToUntruncatedText
  (params (Payload_Text_2_563E0DA1490D74CEE84EF39A5DF6BB26 :type Text) (Payload_Message_5_DDD38AFE499B5AE0E2151A99B81EE564 :type Localizable%20Message%20Structure) (Payload_CharLimit_11_F93CC8954D727984A85AACB9929A1685 :type Integer))
  (return (select (LocalizableMessage|IsEmptyLocalizableMessage Payload_Message_5_DDD38AFE499B5AE0E2151A99B81EE564) Payload_Text_2_563E0DA1490D74CEE84EF39A5DF6BB26 (LocalizableMessage|ConvLocalizableMessagetoText (Variables|Get__WorldContext) Payload_Message_5_DDD38AFE499B5AE0E2151A99B81EE564))))

(fn TextToLocalizablePayload
  (params (Text :type Text))
  (return (Utilities|Struct|MakeLocalizablePayload Text (Utilities|Struct|MakeLocalizableMessage))))
