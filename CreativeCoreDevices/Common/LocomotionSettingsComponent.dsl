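; Creative device component that applies a locomotion preset struct to movement settings: energy, sprinting, mantling, sliding, flight, roll landing, fall damage and wall moves.
; SetLocomotionValues writes each field into a ScalableFloat with context, using helpers for roll-landing recovery and skill.
; Wall kick from scramble requires both features; preset kick strength feeds vertical strength, height feeds horizontal.
; The data registry resolver scope comes from the owning actor's parent implementation.
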(fn BuildDataRegistryResolverScope
  (bind (_resolverscopeoutputs_resolverscopes_3_2e33f25e48b4eee38b3443b96ef43d94 _resolverscopeoutputs_priority_7_7799a38b4e4926ac296dbaa5365c7a8a _resolverscopeoutputs_bisvalid_10_2ed2f4c5427065d8522f07a959958412) (Utilities|Struct|BreakDataRegistryResolverScopeOutputs (Variables|GetResolverScopeOutputs_ResolverScopes_3_2E33F25E48B4EEE38B3443B96EF43D94)))
  (Utilities|FlowControl|Sequence
    (:then_0
      (bind _asfort_data_registry_resolver_scope_interface (Utilities|Casting|CastToFortDataRegistryResolverScopeInterface (Components|GetOwner)))
      (bind (_inoutresolverscopes _inoutpriority _returnvalue) (DataRegistry|NativeParentBuildDataRegistryResolverScope _asfort_data_registry_resolver_scope_interface))
      (Variables|SetResolverScopeOutputs_ResolverScopes_3_2E33F25E48B4EEE38B3443B96EF43D94 (Utilities|Struct|MakeDataRegistryResolverScopeOutputs _inoutresolverscopes _inoutpriority _returnvalue)))
    (:then_1
      (return _resolverscopeoutputs_resolverscopes_3_2e33f25e48b4eee38b3443b96ef43d94 _resolverscopeoutputs_priority_7_7799a38b4e4926ac296dbaa5365c7a8a _resolverscopeoutputs_bisvalid_10_2ed2f4c5427065d8522f07a959958412))))

(fn GetCacheContextOverride
  (return 0))

(fn SetLocomotionValues
  (params (LocomotionValues :type Locomotion%20Preset%20Struct%20Structure%20%28by%20ref%29))
  (bind _locomotionvalues (Variables|GetLocomotionValues))
  (bind (_bapplyjumppenalty_7_3a31e3cc4bdc012a0163b1964bbea3df _bgliderredeploy_39_71149b6c4ba6a42bc9edb88694f6ae10 _bflyenabled_41_d15e6f05421ac053006db19a9fa624b6 _ballowflightsprint_43_adf9208b4cce05f589f691ae8ecc9d8a _flyspeedpreset_46_661d89c44c716362e5926c8bf40f5b0e) (Utilities|Struct|BreakLocomotionPresetStruct _locomotionvalues))
  (bind _locomotionvalues_2 (Variables|GetLocomotionValues))
  (bind (_energymax_70_ace05d3b4287ca64b9cdb896123d06ea _energyrechargeamount_68_e905827e4fa43c44b8ba25ba1cc19e40 _energyrechargedelay_69_3c4ad7a94a3cf317dc7c1e8d9e25bb7b) (Utilities|Struct|BreakLocomotionPresetStruct _locomotionvalues_2))
  (bind _locomotionvalues_3 (Variables|GetLocomotionValues))
  (bind (_ballowboostedjump_55_d7cd042547e6720fc6558191fb842532 _boostedjumpenergycost_123_4e6bcc244bd245e853b5dd8af31d163a) (Utilities|Struct|BreakLocomotionPresetStruct _locomotionvalues_3))
  (bind _locomotionvalues_4 (Variables|GetLocomotionValues))
  (bind (_ballowsprinting_23_30bcf21447b1ae2a9b8400a9a6dd74ae _sprintingjumpenergycost_89_a71202eb4aaddaa1b56c8e8e10ac2b25 _sprintingenergycostpersecond_26_5efd167445c4bf5add1922a981a67e8c _maxsprintingjumpmultiplier_28_72abf30642d1c5cdfe9c158b5dd345b2 _maxsprintingspeedmultiplier_30_7d1fd1024020fdf09b7efa85b01ff381) (Utilities|Struct|BreakLocomotionPresetStruct _locomotionvalues_4))
  (bind _locomotionvalues_5 (Variables|GetLocomotionValues))
  (bind (_ballowmantling_9_2d7792d74b333533d42486b460884d0d _ballowvaulting_22_2035cf0f4b9dcc4ba5a877af068194d5 _ballowsliding_33_f2350fb54315154e33ef6f9e39329ed0 _ballowslidekick_35_e880fdbf48f7627124a390904aaafba5 _ballowshoulderbashing_37_30c279db4e82f248338b8aa94166b1f7) (Utilities|Struct|BreakLocomotionPresetStruct _locomotionvalues_5))
  (bind _locomotionvalues_6 (Variables|GetLocomotionValues))
  (bind (_ballowrolllanding_57_0c837ec94f51848ac25f9e8693db244c _rolllandingenergyrechargeamount_94_5f14e4354e6a687804564182c3951d10 _rolllandingenergyrechargefatigue_136_c11d8ca24cab12981f77788363f8ff04 _rolllandingfalldamagereduction_93_7fb8ec454fdcc1560caca6880d1fce44 _rolllandingskill_97_f177e48240b3c5a2ba5aaa9428b8f6ec _rolllandinghighskillwindow_125_f0268c284a5f202a3e82249d2043b163) (Utilities|Struct|BreakLocomotionPresetStruct _locomotionvalues_6))
  (bind _locomotionvalues_7 (Variables|GetLocomotionValues))
  (bind (_ballowwallkick_59_360afa9b4861125509db57a13e811574 _maximumwallkickcount_111_0ebc293f4882863487905bb1578591c0 _wallkickstrength_114_813dc4a049d08b89286f0dad33578348 _wallkickheight_116_f6468e8a4a749f3fa16b90a89a48582c _ballowwallscramble_61_1ff1ad0946ce324f6099a7b54f01e691) (Utilities|Struct|BreakLocomotionPresetStruct _locomotionvalues_7))
  (bind _locomotionvalues_8 (Variables|GetLocomotionValues))
  (bind (_bshouldcapfalldamage_85_b3be7a61451a54d1a3124e8538e046ad _fatalfalldamageshieldcost_132_4c3a61e74b4382c6280c6fb6e5e6529b _bfatalfalldamageragdoll_120_1ec78a1645207408ce85d591c53513ea _bfatalfalldamagestagger_121_dace57554051f08cc73a29ba0437f036 _bfatalfalldamagedrainsenergy_130_0504ff814a20e3a3368b0b9ea4036532 _ballowgroundrush_86_b788f2a546029e9618308abac5cbbe81) (Utilities|Struct|BreakLocomotionPresetStruct _locomotionvalues_8))
  ;[0.1479 1 0.4018 0.5] Energy Settings
  (ScalableFloat|SetValuewithContext (Variables|Energy|GetEnergyMax) _energymax_70_ace05d3b4287ca64b9cdb896123d06ea)
  (ScalableFloat|SetValuewithContext (Variables|Energy|GetEnergyRechargeAmount) _energyrechargeamount_68_e905827e4fa43c44b8ba25ba1cc19e40)
  (ScalableFloat|SetValuewithContext (Variables|Energy|GetEnergyRechargeDelay) _energyrechargedelay_69_3c4ad7a94a3cf317dc7c1e8d9e25bb7b)
  ;[1 0.2351 0.8332 0.5] Locomotion Features
  (ScalableFloat|SetValuewithContext (Variables|Locomotion|Mantling|GetAllowMantling) _ballowmantling_9_2d7792d74b333533d42486b460884d0d)
  (ScalableFloat|SetValuewithContext (Variables|Locomotion|Vaulting|GetAllowVaulting) _ballowvaulting_22_2035cf0f4b9dcc4ba5a877af068194d5)
  (ScalableFloat|SetValuewithContext (Variables|Locomotion|Sliding|GetAllowSliding) _ballowsliding_33_f2350fb54315154e33ef6f9e39329ed0)
  (ScalableFloat|SetValuewithContext (Variables|Locomotion|GetAllowSlideKick) _ballowslidekick_35_e880fdbf48f7627124a390904aaafba5)
  (ScalableFloat|SetValuewithContext (Variables|Locomotion|GetAllowShoulderBashing) _ballowshoulderbashing_37_30c279db4e82f248338b8aa94166b1f7)
  ;[0.005556 0.274 1 0.5] Tactical Sprint
  (ScalableFloat|SetValuewithContext (Variables|Locomotion|Sprinting|GetAllowSprinting) _ballowsprinting_23_30bcf21447b1ae2a9b8400a9a6dd74ae)
  (ScalableFloat|SetValuewithContext (Variables|Locomotion|Sprinting|GetSprintJumpEnergyCost) _sprintingjumpenergycost_89_a71202eb4aaddaa1b56c8e8e10ac2b25)
  (ScalableFloat|SetValuewithContext (Variables|Locomotion|Sprinting|GetSprintingEnergyCostPerSecond) _sprintingenergycostpersecond_26_5efd167445c4bf5add1922a981a67e8c)
  (ScalableFloat|SetValuewithContext (Variables|Locomotion|Sprinting|GetMaxSprintingJumpMultiplier) _maxsprintingjumpmultiplier_28_72abf30642d1c5cdfe9c158b5dd345b2)
  (ScalableFloat|SetValuewithContext (Variables|Locomotion|Sprinting|GetMaxSprintingSpeedMultiplier) _maxsprintingspeedmultiplier_30_7d1fd1024020fdf09b7efa85b01ff381)
  ;[1 0.1058 0.3874 0.5] Gliding & Flight
  (ScalableFloat|SetValuewithContext (Variables|Locomotion|Jumping|GetApplyJumpPenalty) _bapplyjumppenalty_7_3a31e3cc4bdc012a0163b1964bbea3df)
  (ScalableFloat|SetValuewithContext (Variables|Locomotion|Gliding|GetGliderRedeploy) _bgliderredeploy_39_71149b6c4ba6a42bc9edb88694f6ae10)
  (ScalableFloat|SetValuewithContext (Variables|Locomotion|Flying|GetFlyEnabled) _bflyenabled_41_d15e6f05421ac053006db19a9fa624b6)
  (ScalableFloat|SetValuewithContext (Variables|Locomotion|Flying|GetAllowFlightSprint) _ballowflightsprint_43_adf9208b4cce05f589f691ae8ecc9d8a)
  (ScalableFloat|SetValuewithContext (Variables|Locomotion|Flying|GetFlySpeed) _flyspeedpreset_46_661d89c44c716362e5926c8bf40f5b0e)
  ;[0.1116 0.1544 1 0.5] Boosted Jump
  (ScalableFloat|SetValuewithContext (Variables|Locomotion|BoostJumping|GetAllowBoostedJump) _ballowboostedjump_55_d7cd042547e6720fc6558191fb842532)
  (ScalableFloat|SetValuewithContext (Variables|Locomotion|BoostJumping|GetBoostedJumpEnergyCost) _boostedjumpenergycost_123_4e6bcc244bd245e853b5dd8af31d163a)
  ;[0.1479 1 0.4018 0.5] Roll Landing
  (ScalableFloat|SetValuewithContext (Variables|Locomotion|RollLanding|GetAllowRollLanding) _ballowrolllanding_57_0c837ec94f51848ac25f9e8693db244c)
  (Utilities|SetRollLandingEnergyRecovery _rolllandingenergyrechargeamount_94_5f14e4354e6a687804564182c3951d10 _rolllandingenergyrechargefatigue_136_c11d8ca24cab12981f77788363f8ff04
    (:Out
      (ScalableFloat|SetValuewithContext (Variables|Locomotion|RollLanding|GetRollLandingDamageReduction) _rolllandingfalldamagereduction_93_7fb8ec454fdcc1560caca6880d1fce44)
      (Utilities|SetRollLandingSkill _rolllandingskill_97_f177e48240b3c5a2ba5aaa9428b8f6ec
        (:Out
          (ScalableFloat|SetValuewithContext (Variables|Locomotion|RollLanding|GetRollLandingSkillWindow) _rolllandinghighskillwindow_125_f0268c284a5f202a3e82249d2043b163)
          ;[1 0.2351 0.8332 0.5] Fall Damage
          (ScalableFloat|SetValuewithContext (Variables|Locomotion|Falling|GetAllowFallDamageCapping) _bshouldcapfalldamage_85_b3be7a61451a54d1a3124e8538e046ad)
          (ScalableFloat|SetValuewithContext (Variables|Locomotion|Falling|GetFatalFallDamageShieldCost) _fatalfalldamageshieldcost_132_4c3a61e74b4382c6280c6fb6e5e6529b)
          (ScalableFloat|SetValuewithContext (Variables|Locomotion|Falling|GetAllowFatalFallDamageRagdoll) _bfatalfalldamageragdoll_120_1ec78a1645207408ce85d591c53513ea)
          (ScalableFloat|SetValuewithContext (Variables|Locomotion|Falling|GetAllowFatalFallDamageStagger) _bfatalfalldamagestagger_121_dace57554051f08cc73a29ba0437f036)
          (ScalableFloat|SetValuewithContext (Variables|Locomotion|Falling|GetAllowFatalFallDamageEnergyDrain) _bfatalfalldamagedrainsenergy_130_0504ff814a20e3a3368b0b9ea4036532)
          (ScalableFloat|SetValuewithContext (Variables|Locomotion|Falling|GetAllowGroundRush) _ballowgroundrush_86_b788f2a546029e9618308abac5cbbe81)
          ;[0.1116 0.1544 1 0.5] Wall
          (ScalableFloat|SetValuewithContext (Variables|Locomotion|WallKicking|GetAllowWallKick) _ballowwallkick_59_360afa9b4861125509db57a13e811574)
          (ScalableFloat|SetValuewithContext (Variables|Locomotion|WallScrambling|GetAllowWallScramble) _ballowwallscramble_61_1ff1ad0946ce324f6099a7b54f01e691)
          (ScalableFloat|SetValuewithContext (Variables|Default|GetAllowWallKickFromWallScramble) (and _ballowwallkick_59_360afa9b4861125509db57a13e811574 _ballowwallscramble_61_1ff1ad0946ce324f6099a7b54f01e691))
          (ScalableFloat|SetValuewithContext (Variables|Default|GetWallKickFatigueJumpCountHandle) _maximumwallkickcount_111_0ebc293f4882863487905bb1578591c0)
          (ScalableFloat|SetValuewithContext (Variables|Default|GetWallKickVerticalStrengthHandle) _wallkickstrength_114_813dc4a049d08b89286f0dad33578348)
          (ScalableFloat|SetValuewithContext (Variables|Default|GetWallKickHorizontalStrengthHandle) _wallkickheight_116_f6468e8a4a749f3fa16b90a89a48582c))))))
